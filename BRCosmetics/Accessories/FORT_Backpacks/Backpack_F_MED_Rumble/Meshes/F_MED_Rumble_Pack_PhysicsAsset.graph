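# Physics asset for the Rumble backpack cosmetic, giving its small attached items (bear, llama, peely, teriyaki, gloves and heads) secondary sway.
# Each dyn_ bone has a capsule or sphere body, constrained to a shared simspace bone with narrow swing and very tight twist limits.
# The simspace body is four thin boxes forming a floor and three walls around the items, and connects to the pack's box body with free angular motion.
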PhysicsAsset /BRCosmetics/Accessories/FORT_Backpacks/Backpack_F_MED_Rumble/Meshes/F_MED_Rumble_Pack_PhysicsAsset
bodies {
 dyn_bear [{shapeType=Capsule, length=3.6284, radius=2.9033, shapeName=dyn_bear_capsule, center=<3.7936 0.5 0.5>, rotation=<rot -90 180 180>}]
 dyn_glove_c [{
   shapeType=Capsule
   length=2.8893
   radius=2.5139
   shapeName=dyn_glove_c_capsule
   center=<7.4563 1.5 -0.0273>
   rotation=<rot 84.9999 179.9988 179.9988>
  }]
 dyn_glove_l [{
   shapeType=Capsule
   length=5.181
   radius=2.1925
   shapeName=dyn_glove_l_capsule
   center=<3.8686 -0.6583 0.2021>
   rotation=<rot -90 84.2837 -84.2838>
  }]
 dyn_glove_r [{shapeType=Capsule, length=3.2958, radius=2.2042, shapeName=dyn_glove_r_capsule, center=<4.5 0.8 0>, rotation=<rot 90 0 0>}]
 dyn_head_c [{shapeType=Sphere, radius=3.3152, shapeName=dyn_head_c_sphere, center=<2.5628 -0.5639 -0.0001>}]
 dyn_head_r [{shapeType=Capsule, length=3.2732, radius=2.0312, shapeName=dyn_head_r_capsule, center=<5 -0.4419 -0.2867>, rotation=<rot 90 0 0>}]
 dyn_llma [{shapeType=Capsule, length=5.9529, radius=2.3708, shapeName=dyn_llma_capsule, center=<5.3716 0.0001 0>, rotation=<rot -90 85.0623 -85.0625>}]
 dyn_peely [{
   shapeType=Capsule
   length=6.8769
   radius=1.9222
   shapeName=dyn_peely_capsule
   center=<6.6048 -0.2601 0.9563>
   rotation=<rot 90 14.0363 19.0362>
  }]
 dyn_teriyaki [{shapeType=Capsule, length=1.0828, radius=2.6631, shapeName=dyn_teriyaki_capsule, center=<2.321 -0.31 -0.9098>, rotation=<rot -90 0 0>}]
 pack [{shapeType=Box, extentX=18.9417, extentY=3.9987, extentZ=21.453, shapeName=pack_box, center=<0 1.424 4.4721>, rotation=<rot 0 0 5.0002>}]
 simspace [
  {shapeType=Box, extentX=22.767, extentY=16.6069, extentZ=1.4039, shapeName=simspace_box, center=<0 0 0.7154>}
  {shapeType=Box, extentX=0.9573, extentY=17.401, extentZ=19.6132, shapeName=simspace_box, center=<11.4087 0 10.5921>}
  {shapeType=Box, extentX=1.5214, extentY=19.5685, extentZ=19.279, shapeName=simspace_box, center=<-11.3502 0.4417 10.5075>}
  {shapeType=Box, extentX=23.738, extentY=1.6359, extentZ=18.9834, shapeName=simspace_box, center=<0 -7.6822 10.575>}
 ]
}
constraints [
 {bone1Name=dyn_peely, bone2Name=simspace, swing1LimitDegrees=10, swing2LimitDegrees=20, twistLimitDegrees=0.8}
 {bone1Name=dyn_llma, bone2Name=simspace, swing1LimitDegrees=8, swing2LimitDegrees=20, twistLimitDegrees=0.5}
 {bone1Name=dyn_bear, bone2Name=simspace, swing1LimitDegrees=2, swing2LimitDegrees=20, twistLimitDegrees=0.2}
 {bone1Name=dyn_glove_l, bone2Name=simspace, swing1LimitDegrees=20, swing2LimitDegrees=15, twistLimitDegrees=0.2}
 {bone1Name=dyn_teriyaki, bone2Name=simspace, swing1LimitDegrees=5, swing2LimitDegrees=10, twistLimitDegrees=0.5}
 {bone1Name=dyn_glove_c, bone2Name=simspace, swing1LimitDegrees=10, swing2LimitDegrees=5, twistLimitDegrees=0.5}
 {bone1Name=dyn_head_r, bone2Name=simspace, swing1LimitDegrees=15, swing2LimitDegrees=10, twistLimitDegrees=1.5}
 {bone1Name=dyn_glove_r, bone2Name=simspace, swing1LimitDegrees=8, swing2LimitDegrees=24.96, twistLimitDegrees=1}
 {bone1Name=dyn_head_c, bone2Name=simspace, swing1LimitDegrees=20.08, swing2LimitDegrees=20, twistLimitDegrees=0.5}
 {
  bone1Name=simspace
  bone2Name=pack
  swing1LimitDegrees=45
  swing1Motion=Free
  swing2LimitDegrees=45
  swing2Motion=Free
  twistLimitDegrees=45
  twistMotion=Free
 }
]
