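# Physics asset for the F_MED_Rumble_Pack_Sky backpack cosmetic, providing bodies for its dangling "dyn_" pieces such as the bear, llma, peely, teriyaki, gloves and heads.
# Each piece is a single simple shape constrained to the simspace body with tight twist, with wider swing on the gloves and dyn_head_c.
# Simspace is a floor plus three box walls, linked to the two-box pack body with free swing and twist.
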PhysicsAsset /BRCosmetics/Accessories/FORT_Backpacks/Backpack_F_MED_Rumble/Meshes/F_MED_Rumble_Pack_SkyPhysicsAsset
bodies {
 dyn_bear [{shapeType=Capsule, length=3.6435, radius=2.2184, shapeName=dyn_bear_capsule, center=<3.5148 0.5 0.5>, rotation=<rot -90 0 0>}]
 dyn_glove_c [{
   shapeType=Capsule
   length=3.2298
   radius=2.5639
   shapeName=dyn_glove_c_capsule
   center=<6.4202 1.5 -0.1179>
   rotation=<rot 84.9999 179.9988 179.9988>
  }]
 dyn_glove_l [{shapeType=Box, extentX=7.5, extentY=2.9425, extentZ=4.565, shapeName=dyn_glove_l_box, center=<3.8686 -0.6583 -0.2994>}]
 dyn_glove_r [{shapeType=Capsule, length=3.2958, radius=2.2042, shapeName=dyn_glove_r_capsule, center=<4.5 0.8 0>, rotation=<rot 90 0 0>}]
 dyn_head_c [{shapeType=Sphere, radius=3.3152, shapeName=dyn_head_c_sphere, center=<2.5628 -0.5639 -0.0001>}]
 dyn_head_r [{shapeType=Capsule, length=3.2882, radius=1.3463, shapeName=dyn_head_r_capsule, center=<5 -0.4419 -0.2867>, rotation=<rot 90 0 0>}]
 dyn_llma [{shapeType=Box, extentX=8.0675, extentY=1.5733, extentZ=4.6133, shapeName=dyn_llma_box, center=<5.3716 0.0001 0>}]
 dyn_peely [{shapeType=Capsule, length=7.0626, radius=1.4903, shapeName=dyn_peely_capsule, center=<6.9582 -0.7295 0.3592>, rotation=<rot 90 180 -175>}]
 dyn_teriyaki [{shapeType=Capsule, length=2.0151, radius=2.3152, shapeName=dyn_teriyaki_capsule, center=<2 -0.31 -0.5>, rotation=<rot -90 0 0>}]
 pack [
  {shapeType=Box, extentX=35.6167, extentY=3.9987, extentZ=21.453, shapeName=pack_box, center=<0 0.9339 4.515>, rotation=<rot 0 0 5.0002>}
  {shapeType=Box, extentX=35.7489, extentY=25.3003, extentZ=2.0131, shapeName=pack_box, center=<0.0002 -5.8756 13.7852>}
 ]
 simspace [
  {shapeType=Box, extentX=22.767, extentY=16.6069, extentZ=1.4039, shapeName=simspace_box, center=<0 0 0.7154>}
  {shapeType=Box, extentX=1.7597, extentY=20.8773, extentZ=19.6132, shapeName=simspace_box, center=<11.0763 0 10.5921>}
  {shapeType=Box, extentX=1.8426, extentY=25.547, extentZ=19.279, shapeName=simspace_box, center=<-11.3502 0.4417 10.5075>}
  {shapeType=Box, extentX=27.8404, extentY=1.6359, extentZ=18.9834, shapeName=simspace_box, center=<0 -7.6822 10.575>}
 ]
}
constraints [
 {bone1Name=dyn_peely, bone2Name=simspace, swing1LimitDegrees=10, swing2LimitDegrees=10, twistLimitDegrees=1}
 {bone1Name=dyn_llma, bone2Name=simspace, swing1LimitDegrees=10, swing2LimitDegrees=10, twistLimitDegrees=1}
 {bone1Name=dyn_bear, bone2Name=simspace, swing1LimitDegrees=10, swing2LimitDegrees=10, twistLimitDegrees=1}
 {bone1Name=dyn_glove_l, bone2Name=simspace, swing1LimitDegrees=20, swing2LimitDegrees=45, twistLimitDegrees=1}
 {bone1Name=dyn_teriyaki, bone2Name=simspace, swing1LimitDegrees=10, swing2LimitDegrees=10, twistLimitDegrees=1}
 {bone1Name=dyn_glove_c, bone2Name=simspace, swing1LimitDegrees=20, swing2LimitDegrees=45, twistLimitDegrees=1}
 {bone1Name=dyn_head_r, bone2Name=simspace, swing1LimitDegrees=10, swing2LimitDegrees=10, twistLimitDegrees=1}
 {bone1Name=dyn_glove_r, bone2Name=simspace, swing1LimitDegrees=20, swing2LimitDegrees=45, twistLimitDegrees=1}
 {bone1Name=dyn_head_c, bone2Name=simspace, swing1LimitDegrees=20, swing2LimitDegrees=45, twistLimitDegrees=1}
 {
  bone1Name=simspace
  bone2Name=pack
  swing1LimitDegrees=45
  swing1Motion=Free
  swing2LimitDegrees=45
  swing2Motion=Free
  twistLimitDegrees=45
  twistMotion=Free
 }
]
