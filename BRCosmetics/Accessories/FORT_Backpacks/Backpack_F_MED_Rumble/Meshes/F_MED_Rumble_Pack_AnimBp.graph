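# Animation Blueprint for the Rumble backpack cosmetic, adding physics-driven motion to its crane, claws, wires and dangling prize bones.
# A cached Anim Dynamics base pose feeds rigid-body or per-bone spring branches, selected by skydiving, DBNO and rigid-body flags.
# A kill-stat watcher triggers Crane Alpha, weighting claw rotation and a final claw rigid-body pass.
# Random vectors wobble claws in the skydive branch, and targeting disables claw springs.
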AnimBlueprint /BRCosmetics/Accessories/FORT_Backpacks/Backpack_F_MED_Rumble/Meshes/F_MED_Rumble_Pack_AnimBp
target_skeleton=&/F_MED_Rumble_Pack_Skeleton
graphs [
 AnimGraph {nodes [
   Anim Dynamics - Bone: dyn_teriyaki(Alpha=1, ExternalForce="0,0,0") {
    angularDampingOverride=1
    angularSpringConstant=1000
    bGravityOverrideInSimSpace=true
    bLinearSpring=true
    bOverrideAngularDamping=true
    bUseSphericalLimits=true
    gravityScale=0
    linearSpringConstant=75
    simulationSpace=BoneRelative
    alphaBoolBlend {blendOption=Linear}
    boundBone {boneName=dyn_teriyaki}
    componentAppliedLinearAccClamp=<5000 5000 5000>
    componentLinearAccScale=<0.2 0.2 0.7>
    componentLinearVelScale=<0.3 0.3 0.7>
    gravityOverride {z=-980}
    physicsBodyDefinitions [{
      sphereCollisionRadius=0
      boundBone {boneName=dyn_teriyaki}
      boxExtents=<60 60 60>
      constraintSetup {angularLimitsMax=<2 6 3>, angularLimitsMin=<-5 -5 -3>, angularTarget=<0 0 0>, linearAxesMax=<1.35 1 0>, linearAxesMin=<0.5 -1 0>}
      localJointOffset=<-4 0 0>
     }]
    relativeSpaceBone {boneName=pack}
   }
   Blend Poses by bool(bActiveValue=Get bIsInFrontEnd()) {blendType=Linear}
   Local Space Ref Pose()
   n1 = "Transform (Modify) Bone - Bone: crane_base"!(ComponentPose, Alpha=1, Translation="0.000,-4.772,0.000", Rotation="0,0,0") -> (Pose) {
    alpha=0
    rotationMode=BMM_Replace
    rotationSpace=BCS_BoneSpace
    scaleSpace=BCS_BoneSpace
    translationMode=BMM_Replace
    translationSpace=BCS_BoneSpace
    alphaBoolBlend {blendOption=Linear}
    boneToModify {boneName=crane_base}
    translation {y=-4.772}
   }
   n2 = "Transform (Modify) Bone - Bone: cranearm_02"!(ComponentPose=@n1, Alpha=1, Translation="2.292,-0.000,0.000", Rotation="0,0,0") -> (Pose) {
    alpha=0
    rotationMode=BMM_Replace
    rotationSpace=BCS_BoneSpace
    scaleSpace=BCS_BoneSpace
    translationMode=BMM_Replace
    translationSpace=BCS_BoneSpace
    alphaBoolBlend {blendOption=Linear}
    boneToModify {boneName=cranearm_02}
    translation {x=2.292}
   }
   n3 = "Constraint - dyn_head_c (P)"(ComponentPose=@n2, Alpha=1, ConstraintWeights_0=1) {
    alpha=0
    alphaBoolBlend {blendOption=Linear}
    boneToModify {boneName=dyn_head_c}
    constraintSetup [{targetBone {boneName=cranearm_02}}]
    constraintWeights [1]
   }
   "Transform (Modify) Bone - Bone: dyn_head_c"!(ComponentPose=@n3, Alpha=Reroute Node(InputPin:float) -> (OutputPin:float), Translation="0.000,-2.256,1.327", Rotation="0,0,0") -> (Pose) {
    alpha=0
    rotationMode=BMM_Replace
    rotationSpace=BCS_BoneSpace
    scaleSpace=BCS_BoneSpace
    translationMode=BMM_Replace
    translationSpace=BCS_BoneSpace
    alphaBoolBlend {blendOption=Linear}
    boneToModify {boneName=dyn_head_c}
    translation {y=-2.256, z=1.327}
   }
   n4 = RigidBody!(
    ComponentPose=Local To Component(LocalPose=Use cached pose 'Pose'())
    Alpha
    OverrideWorldGravity:Vector=Get GravityOverride()
    ExternalForce
   ) -> (Pose) {
    bClampLinearTranslationLimitToRefPose=true
    bOverrideWorldGravity=true
    overridePhysicsAsset=&/F_MED_Rumble_Pack_SkyPhysicsAsset
    simulationSpace=BaseBoneSpace
    alphaBoolBlend {blendOption=Linear}
    baseBoneRef {boneName=simspace}
    componentAppliedLinearAccClamp=<2000 2000 2000>
    componentLinearAccScale=<0.4 0.85 0.8>
    componentLinearVelScale=<0.4 0.85 0.8>
    overrideWorldGravity {z=-980}
   }
   n5 = "Transform (Modify) Bone - Bone: cranearm_02"!(ComponentPose=@n4, Alpha=1, Translation="-1.794,0.000,-0.000") -> (Pose) {
    alpha=0
    rotationSpace=BCS_BoneSpace
    scaleSpace=BCS_BoneSpace
    translationMode=BMM_Replace
    translationSpace=BCS_BoneSpace
    alphaBoolBlend {blendOption=Linear}
    boneToModify {boneName=cranearm_02}
    translation {x=-1.794}
   }
   n6 = "Transform (Modify) Bone - Bone: crane_claw_l"!(ComponentPose=@n5, Alpha=1, Translation="0,0,0", Rotation=Get Floating_Rotator1() -> (Floating_Rotator1:Rotator)) -> (Pose) {
    alpha=0
    rotationMode=BMM_Replace
    rotationSpace=BCS_BoneSpace
    scaleSpace=BCS_BoneSpace
    translationMode=BMM_Replace
    translationSpace=BCS_BoneSpace
    alphaBoolBlend {blendOption=Linear}
    boneToModify {boneName=crane_claw_l}
   }
   n7 = "Transform (Modify) Bone - Bone: crane_claw_r"!(ComponentPose=@n6, Alpha=1, Translation="0,0,0", Rotation=Get Floating_Rotator1_0() -> (Floating_Rotator1_0:Rotator)) -> (Pose) {
    alpha=0
    rotationMode=BMM_Replace
    rotationSpace=BCS_BoneSpace
    scaleSpace=BCS_BoneSpace
    translationMode=BMM_Replace
    translationSpace=BCS_BoneSpace
    alphaBoolBlend {blendOption=Linear}
    boneToModify {boneName=crane_claw_r}
    rotation {yaw=-0.003}
   }
   n8 = Anim Dynamics - Bone: dyn_bear(ComponentPose=@n7, Alpha=1, ExternalForce="0,0,0") {
    angularDampingOverride=1
    angularSpringConstant=20000
    bGravityOverrideInSimSpace=true
    bLinearSpring=true
    bOverrideAngularDamping=true
    bOverrideLinearDamping=true
    gravityScale=0.5
    linearDampingOverride=1
    linearSpringConstant=75
    simulationSpace=RootRelative
    alphaBoolBlend {blendOption=Linear}
    boundBone {boneName=dyn_bear}
    componentAppliedLinearAccClamp=<3000 3000 3000>
    componentLinearAccScale=<3 5 7>
    componentLinearVelScale=<3 5 7>
    gravityOverride {z=-980}
    physicsBodyDefinitions [{
      sphereCollisionRadius=0
      boundBone {boneName=dyn_bear}
      boxExtents=<20 30 30>
      constraintSetup {angularLimitsMax=<8 5 5>, angularLimitsMin=<-8 -5 -5>, angularTarget=<1 0 0>}
      localJointOffset=<2 0 0>
     }]
    relativeSpaceBone {boneName=simspace}
   }
   n9 = Anim Dynamics - Bone: dyn_glove_l(ComponentPose=@n8, Alpha=1, ExternalForce="0,0,0") {
    angularDampingOverride=1
    angularSpringConstant=20000
    bGravityOverrideInSimSpace=true
    bLinearSpring=true
    bOverrideAngularDamping=true
    bUseSphericalLimits=true
    gravityScale=0.5
    linearSpringConstant=75
    simulationSpace=RootRelative
    alphaBoolBlend {blendOption=Linear}
    boundBone {boneName=dyn_glove_l}
    componentAppliedLinearAccClamp=<3000 3000 3000>
    componentLinearAccScale=<3 5 7>
    componentLinearVelScale=<3 5 7>
    gravityOverride {z=-980}
    physicsBodyDefinitions [{
      sphereCollisionRadius=0
      boundBone {boneName=dyn_glove_l}
      boxExtents=<20 30 30>
      constraintSetup {angularLimitsMax=<8 5 5>, angularLimitsMin=<-8 -5 -5>, angularTarget=<0 0 0>}
      localJointOffset=<2 0 0>
     }]
    relativeSpaceBone {boneName=simspace}
   }
   n10 = Anim Dynamics - Bone: dyn_teriyaki(ComponentPose=@n9, Alpha=1, ExternalForce="0,0,0") {
    angularDampingOverride=1
    angularSpringConstant=20000
    bGravityOverrideInSimSpace=true
    bLinearSpring=true
    bOverrideAngularDamping=true
    bOverrideLinearDamping=true
    gravityScale=0.5
    linearDampingOverride=1
    linearSpringConstant=75
    simulationSpace=RootRelative
    alphaBoolBlend {blendOption=Linear}
    boundBone {boneName=dyn_teriyaki}
    componentAppliedLinearAccClamp=<3000 3000 3000>
    componentLinearAccScale=<3 5 7>
    componentLinearVelScale=<3 5 7>
    gravityOverride {z=-980}
    physicsBodyDefinitions [{
      sphereCollisionRadius=0
      boundBone {boneName=dyn_teriyaki}
      boxExtents=<20 30 30>
      constraintSetup {angularLimitsMax=<8 5 5>, angularLimitsMin=<-8 -5 -5>, angularTarget=<1 0 0>}
      localJointOffset=<2 0 0>
     }]
    relativeSpaceBone {boneName=simspace}
   }
   n11 = Anim Dynamics - Bone: dyn_llma(ComponentPose=@n10, Alpha=1, ExternalForce="0,0,0") {
    angularDampingOverride=1
    angularSpringConstant=20000
    bGravityOverrideInSimSpace=true
    bLinearSpring=true
    bOverrideAngularDamping=true
    bUseSphericalLimits=true
    gravityScale=0.5
    linearSpringConstant=75
    simulationSpace=RootRelative
    alphaBoolBlend {blendOption=Linear}
    boundBone {boneName=dyn_llma}
    componentAppliedLinearAccClamp=<3000 3000 3000>
    componentLinearAccScale=<3 5 7>
    componentLinearVelScale=<3 5 7>
    gravityOverride {z=-980}
    physicsBodyDefinitions [{
      sphereCollisionRadius=0
      boundBone {boneName=dyn_llma}
      boxExtents=<20 30 30>
      constraintSetup {angularLimitsMax=<8 5 5>, angularLimitsMin=<-8 -5 -5>, angularTarget=<0 0 0>}
      localJointOffset=<2 0 0>
     }]
    relativeSpaceBone {boneName=simspace}
   }
   n12 = Anim Dynamics - Bone: dyn_head_c(ComponentPose=@n11, Alpha=1, ExternalForce="0,0,0") {
    angularDampingOverride=1
    angularSpringConstant=20000
    bGravityOverrideInSimSpace=true
    bLinearSpring=true
    bOverrideAngularDamping=true
    bOverrideLinearDamping=true
    gravityScale=0.5
    linearDampingOverride=1
    linearSpringConstant=75
    simulationSpace=RootRelative
    alphaBoolBlend {blendOption=Linear}
    boundBone {boneName=dyn_head_c}
    componentAppliedLinearAccClamp=<3000 3000 3000>
    componentLinearAccScale=<3 5 7>
    componentLinearVelScale=<3 5 7>
    gravityOverride {z=-980}
    physicsBodyDefinitions [{
      sphereCollisionRadius=0
      boundBone {boneName=dyn_head_c}
      boxExtents=<20 30 30>
      constraintSetup {angularLimitsMax=<8 5 5>, angularLimitsMin=<-8 -5 -5>, angularTarget=<1 0 0>}
      localJointOffset=<2 0 0>
     }]
    relativeSpaceBone {boneName=simspace}
   }
   n13 = Anim Dynamics - Bone: dyn_glove_c(ComponentPose=@n12, Alpha=1, ExternalForce="0,0,0") {
    angularDampingOverride=1
    angularSpringConstant=20000
    bGravityOverrideInSimSpace=true
    bLinearSpring=true
    bOverrideAngularDamping=true
    bUseSphericalLimits=true
    gravityScale=0.5
    linearSpringConstant=75
    simulationSpace=RootRelative
    alphaBoolBlend {blendOption=Linear}
    boundBone {boneName=dyn_glove_c}
    componentAppliedLinearAccClamp=<3000 3000 3000>
    componentLinearAccScale=<3 5 7>
    componentLinearVelScale=<3 5 7>
    gravityOverride {z=-980}
    physicsBodyDefinitions [{
      sphereCollisionRadius=0
      boundBone {boneName=dyn_glove_c}
      boxExtents=<20 30 30>
      constraintSetup {angularLimitsMax=<8 5 5>, angularLimitsMin=<-8 -5 -5>, angularTarget=<0 0 0>}
      localJointOffset=<2 0 0>
     }]
    relativeSpaceBone {boneName=simspace}
   }
   n14 = Anim Dynamics - Bone: dyn_peely(ComponentPose=@n13, Alpha=1, ExternalForce="0,0,0") {
    angularDampingOverride=1
    angularSpringConstant=20000
    bGravityOverrideInSimSpace=true
    bLinearSpring=true
    bOverrideAngularDamping=true
    bUseSphericalLimits=true
    gravityScale=0.5
    linearSpringConstant=75
    simulationSpace=RootRelative
    alphaBoolBlend {blendOption=Linear}
    boundBone {boneName=dyn_peely}
    componentAppliedLinearAccClamp=<3000 3000 3000>
    componentLinearAccScale=<3 5 7>
    componentLinearVelScale=<3 5 7>
    gravityOverride {z=-980}
    physicsBodyDefinitions [{
      sphereCollisionRadius=0
      boundBone {boneName=dyn_peely}
      boxExtents=<20 30 30>
      constraintSetup {angularLimitsMax=<8 5 5>, angularLimitsMin=<-8 -5 -5>, angularTarget=<0 0 0>}
      localJointOffset=<2 0 0>
     }]
    relativeSpaceBone {boneName=simspace}
   }
   n15 = Anim Dynamics - Bone: dyn_glove_r(ComponentPose=@n14, Alpha=1.01, ExternalForce="0,0,0") {
    angularDampingOverride=1
    angularSpringConstant=20000
    bGravityOverrideInSimSpace=true
    bLinearSpring=true
    bOverrideAngularDamping=true
    bUseSphericalLimits=true
    gravityScale=0.5
    linearSpringConstant=75
    simulationSpace=RootRelative
    alphaBoolBlend {blendOption=Linear}
    boundBone {boneName=dyn_glove_r}
    componentAppliedLinearAccClamp=<3000 3000 3000>
    componentLinearAccScale=<3 5 7>
    componentLinearVelScale=<3 5 7>
    gravityOverride {z=-980}
    physicsBodyDefinitions [{
      sphereCollisionRadius=0
      boundBone {boneName=dyn_glove_r}
      boxExtents=<20 30 30>
      constraintSetup {angularLimitsMax=<8 5 5>, angularLimitsMin=<-8 -5 -5>, angularTarget=<0 0 0>}
      localJointOffset=<2 0 0>
     }]
    relativeSpaceBone {boneName=simspace}
   }
   n16 = Anim Dynamics - Bone: dyn_head_r(ComponentPose=@n15, Alpha=1, ExternalForce="0,0,0") {
    angularDampingOverride=1
    angularSpringConstant=20000
    bGravityOverrideInSimSpace=true
    bLinearSpring=true
    bOverrideAngularDamping=true
    bOverrideLinearDamping=true
    gravityScale=0.5
    linearDampingOverride=1
    linearSpringConstant=75
    simulationSpace=RootRelative
    alphaBoolBlend {blendOption=Linear}
    boundBone {boneName=dyn_head_r}
    componentAppliedLinearAccClamp=<3000 3000 3000>
    componentLinearAccScale=<3 5 7>
    componentLinearVelScale=<3 5 7>
    gravityOverride {z=-980}
    physicsBodyDefinitions [{
      sphereCollisionRadius=0
      boundBone {boneName=dyn_head_r}
      boxExtents=<20 30 30>
      constraintSetup {angularLimitsMax=<8 5 5>, angularLimitsMin=<-8 -5 -5>, angularTarget=<1 0 0>}
      localJointOffset=<2 0 0>
     }]
    relativeSpaceBone {boneName=simspace}
   }
   n17 = Component To Local(ComponentPose=@n16)
   n18 = Use cached pose 'Dynamics'()
   n19 = Get bIsSkyDiving()
   n20 = RigidBody!(
    ComponentPose=Local To Component(LocalPose=Blend Poses by bool(BlendPose_0=@n17, BlendPose_1=@n18, bActiveValue=@n19) {blendType=Linear})
    Alpha=Get Crane Alpha() -> (Crane Alpha:double)
    ExternalForce
   ) -> (Pose) {
    bClampLinearTranslationLimitToRefPose=true
    bOverrideWorldGravity=true
    overridePhysicsAsset=&/F_MED_Rumble_Pack__ClawPhysics
    simulationSpace=BaseBoneSpace
    alphaBoolBlend {blendOption=Linear}
    alphaScaleBiasClamp {bInterpResult=true, interpSpeedIncreasing=20}
    baseBoneRef {boneName=root}
    componentAppliedLinearAccClamp=<3000 3000 3000>
    componentLinearAccScale=<0.2 0.3 0.6>
    componentLinearVelScale=<0.2 0.3 0.6>
    overrideWorldGravity {z=-980}
   }
   n21 = "Transform (Modify) Bone - Bone: crane_base"(ComponentPose=@n20, Alpha=1, Translation="0.000,-4.5,0.000", Rotation="0,0,0") {
    alpha=0
    rotationMode=BMM_Replace
    rotationSpace=BCS_BoneSpace
    scaleSpace=BCS_BoneSpace
    translationMode=BMM_Replace
    translationSpace=BCS_BoneSpace
    alphaBoolBlend {blendOption=Linear}
    alphaScaleBiasClamp {bInterpResult=true, interpSpeedDecreasing=5, interpSpeedIncreasing=25}
    boneToModify {boneName=crane_base}
    translation {y=-4.5}
   }
   Output Pose(Result=Component To Local(ComponentPose=@n21)) {layerGroup=None}
   n22 = RigidBody(ComponentPose=Local To Component(LocalPose=Use cached pose 'Pose'())) {
    bClampLinearTranslationLimitToRefPose=true
    bOverrideWorldGravity=true
    overridePhysicsAsset=&/F_MED_Rumble_Pack_PhysicsAsset
    simulationSpace=BaseBoneSpace
    alphaBoolBlend {blendOption=Linear}
    baseBoneRef {boneName=root}
    componentAppliedLinearAccClamp=<3000 3000 3000>
    componentLinearAccScale=<0.2 0.3 0.6>
    componentLinearVelScale=<0.2 0.3 0.6>
    overrideWorldGravity {z=-980}
   }
   n23 = Component To Local(ComponentPose=@n22)
   n24 = RigidBody!(ComponentPose=Local To Component(LocalPose=Use cached pose 'Pose'()), Alpha, OverrideWorldGravity:Vector="0,0,-980", ExternalForce) -> (Pose) {
    bClampLinearTranslationLimitToRefPose=true
    bOverrideWorldGravity=true
    overridePhysicsAsset=&F_MED_Rumble_Pack_SkyPhysicsAsset
    simulationSpace=BaseBoneSpace
    alphaBoolBlend {blendOption=Linear}
    baseBoneRef {boneName=simspace}
    componentAppliedLinearAccClamp=<2000 2000 2000>
    componentLinearAccScale=<0.4 0.85 0.8>
    componentLinearVelScale=<0.4 0.85 0.8>
    overrideWorldGravity {z=-980}
   }
   n25 = Component To Local(ComponentPose=@n24)
   n26 = Get bIsDBNO()
   n27 = Anim Dynamics - Bone: dyn_bear(ComponentPose=Local To Component(LocalPose=Use cached pose 'Pose'()), Alpha=0.3, ExternalForce="0,0,0") {
    angularDampingOverride=1
    angularSpringConstant=20000
    bGravityOverrideInSimSpace=true
    bLinearSpring=true
    bOverrideAngularDamping=true
    bOverrideLinearDamping=true
    gravityScale=0.5
    linearDampingOverride=1
    linearSpringConstant=150
    simulationSpace=BoneRelative
    alphaBoolBlend {blendOption=Linear}
    boundBone {boneName=dyn_bear}
    componentAppliedLinearAccClamp=<3000 3000 3000>
    componentLinearAccScale=<0.3 0.3 0.7>
    componentLinearVelScale=<0.3 0.3 0.7>
    gravityOverride {z=-980}
    physicsBodyDefinitions [{
      sphereCollisionRadius=0
      boundBone {boneName=dyn_bear}
      boxExtents=<20 30 30>
      constraintSetup {angularLimitsMax=<4 3 3>, angularLimitsMin=<-2 -1 -1>, angularTarget=<1 0 0>, linearAxesMax=<2.5 1 1.5>, linearAxesMin=<-0.2 -0.5 0.3>}
      localJointOffset=<0 0 0>
     }]
    relativeSpaceBone {boneName=pack}
   }
   n28 = Anim Dynamics - Bone: dyn_glove_l(ComponentPose=@n27, Alpha=0.8, ExternalForce="0,0,0") {
    angularDampingOverride=1
    angularSpringConstant=1000
    bGravityOverrideInSimSpace=true
    bLinearSpring=true
    bOverrideAngularDamping=true
    bUseSphericalLimits=true
    gravityScale=0
    linearSpringConstant=75
    simulationSpace=BoneRelative
    alphaBoolBlend {blendOption=Linear}
    boundBone {boneName=dyn_glove_l}
    componentAppliedLinearAccClamp=<5000 5000 5000>
    componentLinearAccScale=<0.2 0.3 0.5>
    componentLinearVelScale=<0.2 0.3 0.5>
    gravityOverride {z=-980}
    physicsBodyDefinitions [{
      sphereCollisionRadius=0
      boundBone {boneName=dyn_glove_l}
      boxExtents=<60 60 60>
      constraintSetup {angularLimitsMax=<4 3 3>, angularLimitsMin=<-2 -1 -1>, angularTarget=<0 0 0>, linearAxesMax=<2.5 1 1.5>, linearAxesMin=<-0.2 -0.5 0.3>}
      localJointOffset=<-4 0 0>
     }]
    relativeSpaceBone {boneName=pack}
   }
   n29 = Anim Dynamics - Bone: dyn_teriyaki(ComponentPose=@n28, Alpha=0.3, ExternalForce="0,0,0") {
    angularDampingOverride=1
    angularSpringConstant=1000
    bAngularSpring=true
    bGravityOverrideInSimSpace=true
    bLinearSpring=true
    bOverrideAngularDamping=true
    bOverrideLinearDamping=true
    gravityScale=0.35
    linearDampingOverride=1
    linearSpringConstant=150
    simulationSpace=BoneRelative
    alphaBoolBlend {blendOption=Linear}
    boundBone {boneName=dyn_teriyaki}
    componentAppliedLinearAccClamp=<3000 3000 3000>
    componentLinearAccScale=<0.3 0.3 1>
    componentLinearVelScale=<0.3 0.3 1>
    gravityOverride {z=-980}
    physicsBodyDefinitions [{
      sphereCollisionRadius=0
      boundBone {boneName=dyn_teriyaki}
      boxExtents=<20 30 30>
      constraintSetup {angularLimitsMax=<4 3 3>, angularLimitsMin=<-2 -1 -1>, angularTarget=<1 0 0>, linearAxesMax=<2.5 1 1.5>, linearAxesMin=<-0.2 -0.5 0.3>}
      localJointOffset=<0 0 0>
     }]
    relativeSpaceBone {boneName=pack}
   }
   n30 = Anim Dynamics - Bone: dyn_llma(ComponentPose=@n29, Alpha=0.9, ExternalForce="0,0,0") {
    angularDampingOverride=1
    angularSpringConstant=20000
    bGravityOverrideInSimSpace=true
    bLinearSpring=true
    bOverrideAngularDamping=true
    bUseSphericalLimits=true
    gravityScale=0.4
    linearSpringConstant=75
    simulationSpace=BoneRelative
    alphaBoolBlend {blendOption=Linear}
    boundBone {boneName=dyn_llma}
    componentAppliedLinearAccClamp=<5000 5000 5000>
    componentLinearAccScale=<0.3 0.6 0.7>
    componentLinearVelScale=<0.3 0.6 0.7>
    gravityOverride {z=-980}
    physicsBodyDefinitions [{
      sphereCollisionRadius=0
      boundBone {boneName=dyn_llma}
      boxExtents=<60 60 60>
      constraintSetup {angularLimitsMax=<4 3 3>, angularLimitsMin=<-2 -1 -1>, angularTarget=<0 0 0>, linearAxesMax=<2.5 1 1.5>, linearAxesMin=<-0.2 -0.5 0.3>}
      localJointOffset=<0 0 0>
     }]
    relativeSpaceBone {boneName=pack}
   }
   n31 = Anim Dynamics - Bone: dyn_head_c(ComponentPose=@n30, Alpha=0.3, ExternalForce="0,0,0") {
    angularDampingOverride=1
    angularSpringConstant=1000
    bAngularSpring=true
    bGravityOverrideInSimSpace=true
    bLinearSpring=true
    bOverrideAngularDamping=true
    bOverrideLinearDamping=true
    gravityScale=0.35
    linearDampingOverride=1
    linearSpringConstant=150
    simulationSpace=BoneRelative
    alphaBoolBlend {blendOption=Linear}
    boundBone {boneName=dyn_head_c}
    componentAppliedLinearAccClamp=<3000 3000 3000>
    componentLinearAccScale=<0.3 0.3 1>
    componentLinearVelScale=<0.3 0.3 1>
    gravityOverride {z=-980}
    physicsBodyDefinitions [{
      sphereCollisionRadius=0
      boundBone {boneName=dyn_head_c}
      boxExtents=<20 30 30>
      constraintSetup {angularLimitsMax=<4 3 3>, angularLimitsMin=<-2 -1 -1>, angularTarget=<1 0 0>, linearAxesMax=<2.5 1 1.5>, linearAxesMin=<-0.2 -0.5 0.3>}
      localJointOffset=<0 0 0>
     }]
    relativeSpaceBone {boneName=pack}
   }
   n32 = Anim Dynamics - Bone: dyn_glove_c(ComponentPose=@n31, Alpha=0.9, ExternalForce="0,0,0") {
    angularDampingOverride=1
    angularSpringConstant=1000
    bGravityOverrideInSimSpace=true
    bLinearSpring=true
    bOverrideAngularDamping=true
    bUseSphericalLimits=true
    gravityScale=0
    linearSpringConstant=75
    simulationSpace=BoneRelative
    alphaBoolBlend {blendOption=Linear}
    boundBone {boneName=dyn_glove_c}
    componentAppliedLinearAccClamp=<5000 5000 5000>
    componentLinearAccScale=<0.2 0.3 0.5>
    componentLinearVelScale=<0.2 0.3 0.5>
    gravityOverride {z=-980}
    physicsBodyDefinitions [{
      sphereCollisionRadius=0
      boundBone {boneName=dyn_glove_c}
      boxExtents=<60 60 60>
      constraintSetup {angularLimitsMax=<4 3 3>, angularLimitsMin=<-2 -1 -1>, angularTarget=<0 0 0>, linearAxesMax=<2.5 1 1.5>, linearAxesMin=<-0.2 -0.5 0.3>}
      localJointOffset=<-4 0 0>
     }]
    relativeSpaceBone {boneName=pack}
   }
   n33 = Anim Dynamics - Bone: dyn_peely(ComponentPose=@n32, Alpha=0.7, ExternalForce="0,0,0") {
    angularDampingOverride=1
    angularSpringConstant=1000
    bGravityOverrideInSimSpace=true
    bLinearSpring=true
    bOverrideAngularDamping=true
    bUseSphericalLimits=true
    gravityScale=0
    linearSpringConstant=75
    simulationSpace=BoneRelative
    alphaBoolBlend {blendOption=Linear}
    boundBone {boneName=dyn_peely}
    componentAppliedLinearAccClamp=<5000 5000 5000>
    componentLinearAccScale=<0.2 0.3 0.7>
    componentLinearVelScale=<0.2 0.3 0.7>
    gravityOverride {z=-980}
    physicsBodyDefinitions [{
      sphereCollisionRadius=0
      boundBone {boneName=dyn_peely}
      boxExtents=<30 30 30>
      constraintSetup {angularLimitsMax=<4 3 3>, angularLimitsMin=<-2 -1 -1>, angularTarget=<0 0 0>, linearAxesMax=<2.5 1 1.5>, linearAxesMin=<-0.2 -0.5 0.3>}
      localJointOffset=<0 0 0>
     }]
    relativeSpaceBone {boneName=pack}
   }
   n34 = Anim Dynamics - Bone: dyn_glove_r(ComponentPose=@n33, Alpha=1, ExternalForce="0,0,0") {
    angularDampingOverride=1
    angularSpringConstant=1000
    bGravityOverrideInSimSpace=true
    bLinearSpring=true
    bOverrideAngularDamping=true
    bUseSphericalLimits=true
    gravityScale=0
    linearSpringConstant=75
    simulationSpace=BoneRelative
    alphaBoolBlend {blendOption=Linear}
    boundBone {boneName=dyn_glove_r}
    componentAppliedLinearAccClamp=<5000 5000 5000>
    componentLinearAccScale=<0.2 0.3 0.5>
    componentLinearVelScale=<0.2 0.3 0.5>
    gravityOverride {z=-980}
    physicsBodyDefinitions [{
      sphereCollisionRadius=0
      boundBone {boneName=dyn_glove_r}
      boxExtents=<60 60 60>
      constraintSetup {angularLimitsMax=<4 3 3>, angularLimitsMin=<-2 -1 -1>, angularTarget=<0 0 0>, linearAxesMax=<2.5 1 1.5>, linearAxesMin=<-0.2 -0.5 0.3>}
      localJointOffset=<-4 0 0>
     }]
    relativeSpaceBone {boneName=pack}
   }
   n35 = Anim Dynamics - Bone: dyn_head_r(ComponentPose=@n34, Alpha=1, ExternalForce="0,0,0") {
    angularDampingOverride=1
    angularSpringConstant=1000
    bAngularSpring=true
    bGravityOverrideInSimSpace=true
    bLinearSpring=true
    bOverrideAngularDamping=true
    bOverrideLinearDamping=true
    gravityScale=0.35
    linearDampingOverride=1
    linearSpringConstant=150
    simulationSpace=BoneRelative
    alphaBoolBlend {blendOption=Linear}
    boundBone {boneName=dyn_head_r}
    componentAppliedLinearAccClamp=<3000 3000 3000>
    componentLinearAccScale=<0.3 0.3 0.8>
    componentLinearVelScale=<0.3 0.3 0.8>
    gravityOverride {z=-980}
    physicsBodyDefinitions [{
      sphereCollisionRadius=0
      boundBone {boneName=dyn_head_r}
      boxExtents=<10 20 20>
      constraintSetup {angularLimitsMax=<4 3 3>, angularLimitsMin=<-2 -1 -1>, angularTarget=<1 0 0>, linearAxesMax=<2.5 1 1.5>, linearAxesMin=<-0.2 -0.5 0.3>}
      localJointOffset=<0 0 0>
     }]
    relativeSpaceBone {boneName=pack}
   }
   Save cached pose 'Dynamics'(Pose=Blend Poses by bool(
     BlendPose_0=Blend Poses by bool(BlendPose_0=@n25, BlendPose_1=@n23, bActiveValue=@n26) {blendType=Linear}
     BlendPose_1=Component To Local(ComponentPose=@n35)
     bActiveValue=Get bRigidBodyEnabled()
    ) {blendType=Linear}) {cacheName=Dynamics}
   n36 = "Transform (Modify) Bone - Bone: pack"!(ComponentPose=Local To Component(LocalPose=Input Pose()), Alpha, Translation="0,0,0", Rotation="0,0,0") -> (Pose) {
    alpha=0
    rotationMode=BMM_Replace
    rotationSpace=BCS_BoneSpace
    scaleSpace=BCS_BoneSpace
    translationMode=BMM_Replace
    translationSpace=BCS_BoneSpace
    alphaBoolBlend {blendOption=Linear}
    boneToModify {boneName=pack}
   }
   n37 = Get Crane Alpha()
   n38 = "Transform (Modify) Bone - Bone: crane_claw_l"(ComponentPose=@n36, Alpha=@n37, Translation="0,0,0", Rotation="0,0,150") {
    alpha=0
    rotationMode=BMM_Replace
    rotationSpace=BCS_BoneSpace
    scaleSpace=BCS_BoneSpace
    translationMode=BMM_Replace
    translationSpace=BCS_BoneSpace
    alphaBoolBlend {blendOption=Linear}
    boneToModify {boneName=crane_claw_l}
    rotation {roll=150}
   }
   n39 = "Transform (Modify) Bone - Bone: crane_claw_r"(ComponentPose=@n38, Alpha=@n37, Translation="0,0,0", Rotation="0,0,-140") {
    alpha=0
    rotationMode=BMM_Replace
    rotationSpace=BCS_BoneSpace
    scaleSpace=BCS_BoneSpace
    translationMode=BMM_Replace
    translationSpace=BCS_BoneSpace
    alphaBoolBlend {blendOption=Linear}
    boneToModify {boneName=crane_claw_r}
    rotation {roll=-140}
   }
   n40 = Anim Dynamics - Bone: crane_base(ComponentPose=@n39, Alpha=1, ExternalForce="0,0,0") {
    angularDampingOverride=0.9
    angularSpringConstant=200000
    bAngularSpring=true
    bLinearSpring=true
    bOverrideAngularDamping=true
    bOverrideLinearDamping=true
    gravityScale=0
    linearDampingOverride=0.995
    linearSpringConstant=50
    simulationSpace=BoneRelative
    alphaBoolBlend {blendOption=Linear}
    boundBone {boneName=crane_base}
    componentAppliedLinearAccClamp=<6000 6000 6000>
    componentLinearAccScale=<0.3 0.3 0.1>
    componentLinearVelScale=<0.7 0.7 0.8>
    gravityOverride {z=-980}
    physicsBodyDefinitions [{
      sphereCollisionRadius=0
      boundBone {boneName=crane_base}
      boxExtents=<7 7 7>
      constraintSetup {
       angularTargetAxis=AxisZ
       twistAxis=AxisZ
       angularLimitsMax=<0 0 0>
       angularLimitsMin=<0 0 0>
       angularTarget=<0 0 0>
       linearAxesMax=<0.05 4 0.05>
       linearAxesMin=<-0.02 -10 -0.02>
      }
      localJointOffset=<-0.1 -0.2 0>
     }]
    relativeSpaceBone {boneName=pack}
   }
   n41 = Anim Dynamics - Bone: cranearm_02(ComponentPose=@n40, Alpha=1, ExternalForce="0,0,0") {
    angularDampingOverride=0.9
    angularSpringConstant=200000
    bAngularSpring=true
    bLinearSpring=true
    bOverrideAngularDamping=true
    bOverrideLinearDamping=true
    gravityScale=0
    linearDampingOverride=0.995
    linearSpringConstant=50
    simulationSpace=BoneRelative
    alphaBoolBlend {blendOption=Linear}
    boundBone {boneName=cranearm_02}
    componentAppliedLinearAccClamp=<6000 6000 6000>
    componentLinearAccScale=<0.3 0.3 0.1>
    componentLinearVelScale=<0.7 0.7 0.8>
    gravityOverride {z=-980}
    physicsBodyDefinitions [{
      sphereCollisionRadius=0
      boundBone {boneName=cranearm_02}
      boxExtents=<5 5 5>
      constraintSetup {angularTargetAxis=AxisZ, twistAxis=AxisZ, angularLimitsMax=<8 0 0>, angularLimitsMin=<-5 0 0>, angularTarget=<0 0 0>}
      localJointOffset=<-0.1 -0.2 0>
     }]
    relativeSpaceBone {boneName=pack}
   }
   n42 = Get Target_off() -> (Target_off:double)
   n43 = Anim Dynamics - Bone: crane_claw_l(ComponentPose=@n41, Alpha=@n42, ExternalForce="0,0,0") {
    angularDampingOverride=0.9
    angularSpringConstant=200000
    bAngularSpring=true
    bLinearSpring=true
    bOverrideAngularDamping=true
    bOverrideLinearDamping=true
    gravityScale=0
    linearDampingOverride=0.995
    linearSpringConstant=50
    simulationSpace=BoneRelative
    alphaBoolBlend {blendOption=Linear}
    boundBone {boneName=crane_claw_l}
    componentAppliedLinearAccClamp=<6000 6000 6000>
    componentLinearAccScale=<0.3 0.3 0.1>
    componentLinearVelScale=<0.7 0.7 0.8>
    gravityOverride {z=-980}
    physicsBodyDefinitions [{
      sphereCollisionRadius=0
      boundBone {boneName=crane_claw_l}
      boxExtents=<6 6 3>
      constraintSetup {angularTargetAxis=AxisZ, twistAxis=AxisZ, angularLimitsMax=<30 2 1>, angularLimitsMin=<-20 -2 -1>, angularTarget=<0 0 0>}
      localJointOffset=<0 0 -4>
     }]
    relativeSpaceBone {boneName=pack}
   }
   n44 = Anim Dynamics - Bone: crane_claw_r(ComponentPose=@n43, Alpha=@n42, ExternalForce="0,0,0") {
    angularDampingOverride=0.9
    angularSpringConstant=200000
    bAngularSpring=true
    bLinearSpring=true
    bOverrideAngularDamping=true
    bOverrideLinearDamping=true
    gravityScale=0
    linearDampingOverride=0.995
    linearSpringConstant=50
    simulationSpace=BoneRelative
    alphaBoolBlend {blendOption=Linear}
    boundBone {boneName=crane_claw_r}
    componentAppliedLinearAccClamp=<6000 6000 6000>
    componentLinearAccScale=<0.3 0.3 0.1>
    componentLinearVelScale=<0.7 0.7 0.8>
    gravityOverride {z=-980}
    physicsBodyDefinitions [{
      sphereCollisionRadius=0
      boundBone {boneName=crane_claw_r}
      boxExtents=<6 6 3>
      constraintSetup {angularTargetAxis=AxisZ, twistAxis=AxisZ, angularLimitsMax=<30 2 1>, angularLimitsMin=<-20 -2 -1>, angularTarget=<0 0 0>}
      localJointOffset=<0 0 -4>
     }]
    relativeSpaceBone {boneName=pack}
   }
   n45 = Anim Dynamics - Bone: dyn_wire_top_01(ComponentPose=@n44, Alpha=1, ExternalForce="0,0,0") {
    angularDampingOverride=0.9999
    angularSpringConstant=100
    bAngularSpring=true
    bOverrideAngularDamping=true
    bUseGravityOverride=true
    simulationSpace=RootRelative
    alphaBoolBlend {blendOption=Linear}
    boundBone {boneName=dyn_wire_top_01}
    componentAppliedLinearAccClamp=<5000 5000 5000>
    componentLinearAccScale=<0.3 0.7 1.2>
    componentLinearVelScale=<0.3 0.7 1.2>
    gravityOverride {z=-980}
    physicsBodyDefinitions [{
      sphereCollisionRadius=0
      boundBone {boneName=dyn_wire_top_01}
      boxExtents=<8 4 4>
      constraintSetup {angularLimitsMax=<1 10 15>, angularLimitsMin=<-1 -30 -15>, angularTarget=<0 0 0>}
      localJointOffset=<3 0 0>
     }]
    relativeSpaceBone {boneName=simspace}
   }
   n46 = Anim Dynamics - Bone: dyn_wire_top_02(ComponentPose=@n45, Alpha=1, ExternalForce="0,0,0") {
    angularDampingOverride=0.9999
    angularSpringConstant=100
    bAngularSpring=true
    bOverrideAngularDamping=true
    bUseGravityOverride=true
    simulationSpace=RootRelative
    alphaBoolBlend {blendOption=Linear}
    boundBone {boneName=dyn_wire_top_02}
    componentAppliedLinearAccClamp=<5000 5000 5000>
    componentLinearAccScale=<0.3 0.7 1.2>
    componentLinearVelScale=<0.3 0.7 1.2>
    gravityOverride {z=-980}
    physicsBodyDefinitions [{
      sphereCollisionRadius=0
      boundBone {boneName=dyn_wire_top_02}
      boxExtents=<4 4 8>
      constraintSetup {angularLimitsMax=<1 10 15>, angularLimitsMin=<-1 -30 -15>, angularTarget=<0 0 0>}
      localJointOffset=<4 0 0>
     }]
    relativeSpaceBone {boneName=simspace}
   }
   n47 = Anim Dynamics - Bone: dyn_wire_bot_01(ComponentPose=@n46, Alpha=1, ExternalForce="0,0,0") {
    angularDampingOverride=0.9999
    angularSpringConstant=100
    bAngularSpring=true
    bOverrideAngularDamping=true
    bUseGravityOverride=true
    simulationSpace=RootRelative
    alphaBoolBlend {blendOption=Linear}
    boundBone {boneName=dyn_wire_bot_01}
    componentAppliedLinearAccClamp=<5000 5000 5000>
    componentLinearAccScale=<0.3 0.7 1.2>
    componentLinearVelScale=<0.3 0.7 1.2>
    gravityOverride {z=-980}
    physicsBodyDefinitions [{
      sphereCollisionRadius=0
      boundBone {boneName=dyn_wire_bot_01}
      boxExtents=<8 4 4>
      constraintSetup {angularLimitsMax=<1 5 15>, angularLimitsMin=<-1 -30 -15>, angularTarget=<0 0 0>}
      localJointOffset=<3 0 0>
     }]
    relativeSpaceBone {boneName=simspace}
   }
   n48 = Anim Dynamics - Bone: dyn_wire_bot_02(ComponentPose=@n47, Alpha=1, ExternalForce="0,0,0") {
    angularDampingOverride=0.9999
    angularSpringConstant=100
    bAngularSpring=true
    bOverrideAngularDamping=true
    bUseGravityOverride=true
    simulationSpace=RootRelative
    alphaBoolBlend {blendOption=Linear}
    boundBone {boneName=dyn_wire_bot_02}
    componentAppliedLinearAccClamp=<5000 5000 5000>
    componentLinearAccScale=<0.3 0.7 1.2>
    componentLinearVelScale=<0.3 0.7 1.2>
    gravityOverride {z=-980}
    physicsBodyDefinitions [{
      sphereCollisionRadius=0
      boundBone {boneName=dyn_wire_bot_02}
      boxExtents=<4 4 8>
      constraintSetup {angularLimitsMax=<1 5 15>, angularLimitsMin=<-1 -30 -15>, angularTarget=<0 0 0>}
      localJointOffset=<4 0 0>
     }]
    relativeSpaceBone {boneName=simspace}
   }
   Save cached pose 'Pose'(Pose=Component To Local(ComponentPose=@n48)) {cacheName=Pose}
  ]}
 EventGraph {nodes [
   n1 = Random Vector Function!(
    execute
    CharacterPartAnimInstance:Custom Character Part Anim Instance Object Reference
    SkelMeshComponent:Mesh Component Object Reference
    JointName:Name=None
    GameTimeInSeconds:double=Get customSpeed() -> (customSpeed:double)
    TimeOffset:double=0.5
    PerlinScale:double=0
    Frequency:Vector="0.1,0,0"
    Amplitude:Vector="2,0,0"
    Offset:Vector="0,0,0"
   ) -> (then, VectorOut:Vector)
   > Set Arm(Arm:Vector=@n1) -> (Output_Get:Vector)
   > Set Crane Arm(Crane Arm:Vector="0, 0, 0") -> (Output_Get:Vector)
   > Set Prize Head(Prize Head:Vector="0, 0, 0") -> (Output_Get:Vector)
   > Set Crane Alpha(Crane Alpha:double=0) -> (Output_Get:double)
   > Branch(Condition="Equal (Float)"!(A:double=Get Crane Alpha(), B:double=0) -> (ReturnValue:bool) {is=title K2Node_CallFunction})
   > Set Claw Count(Claw Count:double=0) -> (Output_Get:double)
   > Set Active_Kill(Active_Kill:bool=false) -> (Output_Get:bool)
   > Set Crane Arm(Crane Arm="0, 0, 0")
   > Set Prize Head(Prize Head="0, 0, 0")
   > Set Arm(Arm="0, 0, 0")
   Event BlueprintUpdateAnimation()
   > n2 = Sequence!(execute) -> (then_0, then_1, then_2:Exec, then_3:Exec, then_4:Exec, then_5:Exec, then_6:Exec)
   > n3 = Random Vector Function(JointName=None, GameTimeInSeconds=0, TimeOffset=Get customSpeed(), PerlinScale=0, Frequency="10,0,0", Amplitude="1,0,0", Offset="0,0,0")
   n4 = vector * float(A=@n3, B=30)
   > Set Floating_Rotator1(Floating_Rotator1:Rotator=MakeRotator(Roll=BreakVector(InVec=@n4))) -> (Output_Get:Rotator)
   n5 = Random Vector Function(
    execute=@n2.then_1
    JointName=None
    GameTimeInSeconds=0
    TimeOffset=Get customSpeed()
    PerlinScale=0
    Frequency="-10,0,0"
    Amplitude="-1,0,0"
    Offset="0,0,0"
   )
   n6 = vector * float(A=@n5, B=30)
   > Set Floating_Rotator1_0(Floating_Rotator1_0:Rotator=MakeRotator(Roll=BreakVector(InVec=@n6))) -> (Output_Get:Rotator)
   n7 = Gravity Override State Function DEPRECATED(
    execute=@n2.then_2
    CharacterPartAnimInstance=Self-Reference()
    GameTimeInSeconds=GetGameTimeInSeconds()
    GravityOverrideParams=Get Gravity Override Params()
   )
   > Set GravityOverride(GravityOverride=@n7)
   Set customSpeed(execute=@n2.then_3, customSpeed:double=float * float(A=GetRealTimeSeconds(), B=20) {is=title K2Node_PromotableOperator}) -> (Output_Get:double)
   n8 = Branch(execute=@n2.then_4, Condition=Get bIsTargeting())
   > Set Target_off(Target_off:double=0) -> (Output_Get:double)
   Set Target_off(execute=@n8.else, Target_off=1)
   n9 = Branch(execute=@n2.then_5, Condition=Get Active_Kill() -> (Active_Kill:bool))
   > Set Crane_Timer(Crane_Timer:double=float + float(A=Get Crane_Timer() -> (Crane_Timer:double), B=1) {is=title K2Node_CommutativeAssociativeBinaryOperator}) -> (Output_Get:double)
   n10 = Get Crane_Timer()
   n11 = Get Crane_Timer_Length() -> (Crane_Timer_Length:double)
   > Set Crane Alpha(Crane Alpha=MapRangeClamped(Value=float - float(A=@n11, B=@n10) {is=title K2Node_CallFunction}, InRangeA=Get Crane_Timer_Length(), InRangeB=0, OutRangeA=1, OutRangeB=0))
   > Delay(Duration=1.5)
   > Set Active_Kill(Active_Kill=false)
   Set Crane Alpha(execute=@n9.else, Crane Alpha=0)
   Event OnCharacterCustomizationLoaded()
   > n12 = Sequence!(execute) -> (then_0, then_1)
   > n13 = Cast To FortPlayerPawn(Object=TryGetFortPawn())
   > n14 = WatchCosmeticStat!(
    execute:Exec
    Pawn:Fort Player Pawn Object Reference=@n13
    StatTag:Fort Stat Manager Tag Structure=(TagName="StatManager.AthenaKills")
    DefaultValue:int=0
    bBroadcastInitialValue:bool=true
    InitialValueBroadcastDelay:float=-1
    ChangeValueBroadcastDelay:float=-1
    bConsumeWaitingNotifications:bool=true
   ) -> (then:Exec, OnStatChanged:Exec, StatName:Name, StatValue:int) {is=title K2Node_AsyncAction}
   n15 = Set KillCount(execute=@n14.OnStatChanged, KillCount:double="To Float (Integer)"(InInt=@n14.StatValue)) -> (Output_Get:double)
   > Branch(Condition="float > float"(A=@n15, B=Get Last_KillCount() -> (Last_KillCount:double)) {is=title K2Node_CallFunction})
   > Set Active_Kill(Active_Kill=true)
   > Set Last_KillCount(Last_KillCount:double=@n15) -> (Output_Get:double)
   n16 = Cast To FortPlayerPawn(execute=@n12.then_1, Object=TryGetFortPawn())
   > n17 = WatchCosmeticStat(
    Pawn=@n16
    StatTag=(TagName="StatManager.DamageDealt")
    DefaultValue=0
    bBroadcastInitialValue=true
    InitialValueBroadcastDelay=-1
    ChangeValueBroadcastDelay=-1
    bConsumeWaitingNotifications=true
   )
   n18 = Set DamageCount(execute=@n17.OnStatChanged, DamageCount:double="To Float (Integer)"(InInt=@n17.StatValue)) -> (Output_Get:double)
   > Branch(Condition="float > float"(A=@n18, B=Get Last Damage Count() -> (Last Damage Count:double)))
   > Set Active_Damage(Active_Damage:bool=true) -> (Output_Get:bool)
   > Set Last Damage Count(Last Damage Count:double=@n18) -> (Output_Get:double)
   > Set Claw Count(Claw Count=float + float(A=Get Claw Count() -> (Claw Count:double), B=-180))
   Set Crane Alpha(Crane Alpha=1)
   > Delay(Duration=0.2)
   TryGetPawnOwner()
  ]}
]
variables [
 Active_Damage {type=boolean}
 Active_Kill {type=boolean}
 Arm {type=Vector}
 Claw Count {type=number}
 Crane Alpha {type=number}
 Crane Arm {type=Vector}
 Crane_Timer {type=number}
 Crane_Timer_Length {type=number}
 customSpeed {type=number}
 DamageCount {type=number}
 Floating_Rotator1 {type=Rotator}
 Floating_Rotator1_0 {type=Rotator}
 Gravity Override Params {type=GravityOverrideParamsStruct}
 GravityOverride {type=Vector}
 KillCount {type=number}
 Last Damage Count {type=number}
 Last_KillCount {type=number}
 Prize Head {type=Vector}
 Target_off {type=number}
]
